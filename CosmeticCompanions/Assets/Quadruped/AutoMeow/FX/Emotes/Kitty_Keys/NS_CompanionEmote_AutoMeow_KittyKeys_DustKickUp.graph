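# Niagara system for the dust kick-up effect in the AutoMeow quadruped companion's Kitty Keys emote.
# It holds a single LW_Dust emitter, for which no renderers are listed.
# User parameters supply a foot dust color array and a surface type index, alongside a system module for setting the color array index, suggesting surface-dependent dust tint.
# System scripts also use SystemState, a variable assignment at spawn, and bool-to-float and vector math dynamic inputs.
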NiagaraSystem /CosmeticCompanions/Assets/Quadruped/AutoMeow/FX/Emotes/Kitty_Keys/NS_CompanionEmote_AutoMeow_KittyKeys_DustKickUp
dependencies {
 usedDataInterfaces []
 usedDynamicInputs [
  &MakeCustomFloatFromBool
  &Multiply_Vector
  &VectorFromFloat
 ]
 usedModules [
  &self:SetFootDustFXColorArrayIndex_03
  &self:SystemSpawnScript.SystemScriptSource.SystemScriptGraph.NiagaraNodeAssignment_0.SetVariables_#1
  &SystemState
 ]
}
emitters {LW_Dust {rendererClasses [], renderers []}}
summary {
 systemName=NS_CompanionEmote_AutoMeow_KittyKeys_DustKickUp
 emitters [{emitterName=LW_Dust, rendererClasses []}]
 userVariables [
  User.FootDustColorArray {
   type=NiagaraDataInterfaceArrayColor
   description=""
   defaultValue {
    struct=NiagaraExt_VariableValue_DataInterface
    value {
     dataInterface=&self:NiagaraDataInterfaceArrayColor_0
     dataInterfaceClass=&NiagaraDataInterfaceArrayColor
    }
   }
  }
  User.SurfaceTypeIndex {type=NiagaraInt32}
 ]
}
